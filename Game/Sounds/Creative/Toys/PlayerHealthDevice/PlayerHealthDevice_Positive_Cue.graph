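# Sound cue for the positive feedback sound of the Creative-mode Player Health Device.
# A Random node picks one of three equally weighted wave variations and feeds a Modulator at the root.
# The cue plays at reduced volume under the critical in-game UI sound class, with its own short-range attenuation and concurrency cap in place of shared assets.
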SoundCue /Game/Sounds/Creative/Toys/PlayerHealthDevice/PlayerHealthDevice_Positive_Cue
root=Modulator_0
nodes [
 {id=Modulator_0, children [Random_0]}
 {id=Random_0, children [WavePlayer_0,WavePlayer_1,WavePlayer_2], properties {weights [1,1,1]}}
 {id=WavePlayer_0, properties {soundWaveAssetPtr=&/PlayerHealthDevice_Positive_01}}
 {id=WavePlayer_1, properties {soundWaveAssetPtr=&/PlayerHealthDevice_Positive_02}}
 {id=WavePlayer_2, properties {soundWaveAssetPtr=&/PlayerHealthDevice_Positive_03}}
]
properties {
 bOverrideAttenuation=true
 bOverrideConcurrency=true
 maxDistance=727
 soundClassObject=&Fort_UI_InGame_Critical
 volumeMultiplier=0.3
 attenuationOverrides {falloffDistance=512, attenuationShapeExtents {x=215}}
 concurrencyOverrides {maxCount=12}
}
